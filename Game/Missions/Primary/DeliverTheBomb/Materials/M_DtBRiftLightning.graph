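# Unlit mesh-particle material for rift lightning effects in the Deliver the Bomb mission, usable with Cascade and Niagara.
# A Fresnel blend of two HDR colors, scaled by particle alpha, feeds emissive through an optional time-of-day switch that dims non-day periods.
# A panned noise texture, jumped by stepped particle age and dynamic-parameter offset, drives vertex displacement strongest mid-UV and partly growing with age.
# Added normal-direction thickness shrinks through a fade texture late in life.
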Material /Game/Missions/Primary/DeliverTheBomb/Materials/M_DtBRiftLightning
graph {
 properties {
  bUsedWithMeshParticles=true
  bUsedWithNiagaraMeshParticles=true
  pixelDepthOffsetMode=PDOM_Legacy
  shadingModel=MSM_Unlit
  usedShadingModels=Unlit
 }
 nodes [
  n1 = ScalarParameter() {defaultValue=2, parameterName=FresnelExp}
  n2 = Multiply(
   A=Max(A=ParticleColor().A)
   B=LinearInterpolate(
    A=VectorParameter() {parameterName=Col1, defaultValue=<rgba 11.18 0.4 20 0>}.RGB
    B=VectorParameter() {parameterName=Col2, defaultValue=<rgba 5 2.2 4.9533 0>}.RGB
    Alpha=MaterialFunctionCall(Exponent=@n1) {materialFunction=&MF_Fresnel_FX}.Fresnel
   )
  )
  n3 = Multiply(A=@n2) {constB=0.2}
  MaterialFunctionCall(Emissive=StaticSwitchParameter(
    True=MaterialFunctionCall(Color_Morning=@n3, Color_Day=@n2, Color_Evening=@n3, Color_Night=@n3) {materialFunction=&Color_Time_of_Day}.Result
    False=@n2
   ) {parameterName=Use color of day}) {materialFunction=&MF_FX_EmissiveAdjustments}
  n4 = TextureCoordinate()
  n5 = Abs(None=Subtract(A=ComponentMask(None=@n4) {g=true}) {constB=0.5})
  n6 = OneMinus(None=Power(Base=Multiply(A=@n5) {constB=2}))
  n7 = ParticleRelativeTime()
  n8 = TextureCoordinate() {uTiling=0}
  n9 = DynamicParameter() {defaultValue=<rgba 0 1 1 1>, paramNames [RandomOffset,UVScale,Param3,Param4]}
  n10 = Multiply(A=@n7) {constB=2}
  n11 = Add(A=Subtract(A=@n10, B=Frac(None=@n10)), B=Multiply(A=Time()) {constB=0.3})
  n12 = TextureSample(UVs=Panner(Coordinate=Multiply(A=@n8, B=@n9.UVScale), Time=Add(A=@n11, B=@n9.RandomOffset)) {speedX=-0.17, speedY=-1.3}) {texture=/Game/CharClasses/Outlander/Gadgets/Fragment/ShockTower/Textures/T_ShockTower_LightningOffset_01}
  n13 = AppendVector(A=Multiply(A=@n12.R) {constB=0.5}, B=ComponentMask(None=@n12.RGB) {b=true, g=true})
  n14 = Add(
   A=Multiply(A=Multiply(A=@n7, B=@n13), B=ScalarParameter() {defaultValue=50, parameterName=NormalDeviationStrength}) {constB=50}
   B=Multiply(A=@n12.RGB, B=ScalarParameter() {defaultValue=120, parameterName=TextureDeviationStrength}) {constB=120}
  )
  n15 = Add(A=@n12.R, B=@n12.R)
  n16 = OneMinus(None=Divide(A=Add(A=@n15, B=@n12.B)) {constB=3})
  n17 = Multiply(
   A=Transform(None=VertexNormalWS()) {transformSourceType=TRANSFORMSOURCE_World, transformType=TRANSFORM_Local}
   B=Add(A=Multiply(A=@n16) {constB=2}) {constB=2}
  ) {constB=5}
  n18 = TextureSample(UVs=OneMinus(None=TextureCoordinate())) {texture=/Game/CharClasses/Outlander/Gadgets/Fragment/ShockTower/Textures/T_ShockTower_LightningFade_01}
  n19 = Clamp(None=Add(A=Multiply(A=@n7) {constB=2}) {constB=-1})
  n20 = Clamp(None=Subtract(A=Add(A=@n18.R), B=Multiply(A=@n19) {constB=2}))
  Transform(None=Add(A=Multiply(A=@n6, B=@n14), B=Multiply(A=@n17, B=@n20))) {transformSourceType=TRANSFORMSOURCE_Local}
 ]
}
