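# Unlit translucent base material for Deliver the Bomb rift tendrils, built for mesh particles including Niagara.
# Particle color drives emissive, optionally adjusted by time of day through a static switch, then passed through an FX emissive adjustment function.
# Panning noise displaces vertices along normals for thickness and along local X/Y for wobble, with masks tapering both toward the V ends.
# Dynamic parameters set noise repeat, thickness, wobble strength and speed; WPO_Intensity scales the total offset.
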Material /Game/Missions/Primary/DeliverTheBomb/Materials/M_DtBRiftTendrilBase
graph {
 properties {
  bEnableResponsiveAA=true
  bUsedWithMeshParticles=true
  bUsedWithNiagaraMeshParticles=true
  blendMode=BLEND_Translucent
  pixelDepthOffsetMode=PDOM_Legacy
  previewMesh=/BRCosmetics/Gadgets/Assets/VinderTech_GliderChute/Glider_York_Male/FX/SM_Glider_York_Lightning
  shadingModel=MSM_Unlit
  usedShadingModels=Unlit
 }
 nodes [
  n1 = ParticleColor()
  n2 = Multiply(A=@n1.RGB) {constB=0.25}
  n3 = Multiply(A=@n1.RGB)
  MaterialFunctionCall(Emissive=StaticSwitchParameter(
    True=MaterialFunctionCall(Color_Morning=@n2, Color_Day=@n3, Color_Evening=@n3, Color_Night=@n2) {materialFunction=&Color_Time_of_Day}.Result
    False=@n1.RGB
   ) {parameterName=Color_Time_of_Day}) {materialFunction=&MF_FX_EmissiveAdjustments}
  n4 = VertexNormalWS()
  n5 = TextureCoordinate()
  n6 = MaterialFunctionCall(X=Constant() {r=1}, Y=DynamicParameter() {paramNames [VRepeat, ThicknessScale, Param3, Param4]}.VRepeat) {materialFunction=&MakeFloat2}
  n7 = TextureSample(UVs=Panner(Coordinate=Multiply(A=@n5, B=@n6.Result)) {bFractionalPart=true, speedY=-0.3}) {texture=&T_NoisePack_01}
  n8 = ComponentMask(None=TextureCoordinate()) {g=true}
  n9 = OneMinus(None=Power(Base=@n8) {constExponent=16})
  n10 = OneMinus(None=@n8)
  n11 = OneMinus(None=Power(Base=@n10) {constExponent=16})
  n12 = Multiply(
   A=Multiply(A=Multiply(A=@n7.R) {constB=2}, B=Multiply(A=@n9, B=@n11))
   B=DynamicParameter() {paramNames [VRepeat, ThicknessScale, Param3, Param4]}.ThicknessScale
  )
  n13 = OneMinus(None=Power(Base=@n10) {constExponent=4})
  n14 = TextureCoordinate()
  n15 = OneMinus(None=Power(Base=ComponentMask(None=@n14) {g=true}) {constExponent=8})
  n16 = Multiply(A=Saturate(None=ConstantBiasScale(None=@n13) {bias=0.2, scale=1}), B=Saturate(None=ConstantBiasScale(None=@n15) {bias=-0.5, scale=2}))
  n17 = Constant3Vector() {constant=<rgba 1 0 0 0>}
  n18 = TextureCoordinate() {uTiling=0, vTiling=0.72}
  n19 = Multiply(A=DynamicParameter() {paramNames [VRepeat, ThicknessScale, Param3, Param4]}.Param4, B=Time())
  n20 = ConstantBiasScale(None=TextureSample(UVs=Panner(Coordinate=@n18, Time=@n19) {bFractionalPart=true, speedX=0.1, speedY=-0.37}) {texture=&T_NoisePack_01}.R) {bias=-0.5, scale=2}
  n21 = Multiply(A=DynamicParameter() {paramNames [VRepeat, ThicknessScale, Param3, Param4]}.Param3, B=Constant() {r=20})
  n22 = Constant3Vector() {constant=<rgba 0 1 0 0>}
  n23 = TextureCoordinate() {uTiling=0, vTiling=0.72}
  n24 = ConstantBiasScale(None=TextureSample(UVs=Panner(Coordinate=@n23, Time=@n19) {bFractionalPart=true, speedX=-0.1, speedY=-0.511}) {texture=&T_NoisePack_01}.G) {bias=-0.5, scale=2}
  n25 = Add(
   A=Multiply(A=Transform(None=@n17) {transformSourceType=TRANSFORMSOURCE_Local}, B=Multiply(A=@n20, B=@n21) {constB=20})
   B=Multiply(A=Transform(None=@n22) {transformSourceType=TRANSFORMSOURCE_Local}, B=Multiply(A=@n24, B=@n21) {constB=20})
  )
  Multiply(A=Add(A=Multiply(A=@n4, B=@n12), B=Multiply(A=@n16, B=@n25)), B=ScalarParameter() {defaultValue=1, parameterName=WPO_Intensity})
 ]
}
